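; BP_VenueSpotlight_Rotation is a venue stage spotlight, named for rotation, that reacts to music stage show signals.
; BeginPlay calls Play where cosmetic events can run; activation strength then toggles Play/Stop and light visibility against a small threshold.
; Color shifts set the SpotLight to primary, secondary or tertiary, chosen by ColorSelector.
; Activate and color logic is duplicated in EventStageShow events and MojoStageShow functions; other Mojo hooks are stubs returning false.
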(event EventBeginPlay
  (Networking|CanExecuteCosmeticEvents
    (:True
      (CallFunction|Play))))

(event EventStageShowColorShift
  (params (Primary :type Linear%20Color%20Structure) (Secondary :type Linear%20Color%20Structure) (Tertiary :type Linear%20Color%20Structure) (AltPrimary :type Linear%20Color%20Structure) (AltSecondary :type Linear%20Color%20Structure) (AltTertiary :type Linear%20Color%20Structure) (MusicTimeMeasure :type Float%20%28single-precision%29))
  (Rendering|Components|Light|SetLightColor (Variables|Default|GetSpotLight) (Utilities|Select :Index (Variables|Default|GetColorSelector) :Option%200 Primary :Option%201 Secondary :Option%202 Tertiary) false))

(event EventStageShowActivate
  (params (Strength :type Float%20%28single-precision%29) (MusicTimeMeasure :type Float%20%28single-precision%29))
  (bind _returnvalue (> Strength 0.1))
  (if _returnvalue
    (CallFunction|Play)
    (else
      (CallFunction|Stop)))
  (Rendering|SetVisibility (Variables|Default|GetSpotLight) _returnvalue false))

(fn MojoStageShowActivate
  (params (Strength :type Float%20%28double-precision%29) (MusicTimeMeasure :type Float%20%28double-precision%29))
  (bind _returnvalue (> Strength 0.1))
  (if _returnvalue
    (CallFunction|Play)
    (else
      (CallFunction|Stop)))
  (Rendering|SetVisibility (Variables|Default|GetSpotLight) _returnvalue false)
  (return false))

(fn MojoStageShowActivateBeat
  (params (Active :type Float%20%28double-precision%29) (MusicTimeMeasure :type Float%20%28double-precision%29))
  (return false))

(fn MojoStageShowColorShift
  (params (Primary :type Linear%20Color%20Structure) (Secondary :type Linear%20Color%20Structure) (Tertiary :type Linear%20Color%20Structure) (AltPrimary :type Linear%20Color%20Structure) (AltSecondary :type Linear%20Color%20Structure) (AltTertiary :type Linear%20Color%20Structure) (MusicTimeMeasure :type Float%20%28double-precision%29))
  (Rendering|Components|Light|SetLightColor (Variables|Default|GetSpotLight) (Utilities|Select :Index (Variables|Default|GetColorSelector) :Option%200 Primary :Option%201 Secondary :Option%202 Tertiary) false)
  (return false))

(fn MojoStageShowCustomMode
  (params (ModeIndex :type Integer) (MusicTimeMeasure :type Float%20%28double-precision%29))
  (return false))

(fn MojoStageShowImpulse
  (params (MusicTimeMeasure :type Float%20%28double-precision%29))
  (return false))

(fn MojoStageShowOnStop
  (return false))

(fn MojoStageShowUpdateNiagaraSystem
  (params (Niagara%20System :type Niagara%20System%20Object%20Reference))
  (return false))

(fn ConstructionScript)
