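; Managed venue light Blueprint: a spotlight prop that reacts to music-timed stage show events.
; Construction script applies exposed intensity, cone angle, attenuation radius and shadow settings.
; Color shift picks one of three colors from the main or alternate palette by ColorSelector and shows the light; impulse and beat also show it, while activate thresholds Strength.
; Mojo-prefixed functions duplicate these handlers or forward calls; a delayed-hide event has no caller here.
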(event Custom|DisableLightByTimer
  (Utilities|FlowControl|Delay 0.2)
  (Rendering|SetVisibility (Variables|Default|GetSpotLight) false true))

(event EventStageShowColorShift
  (params (Primary :type Linear%20Color%20Structure) (Secondary :type Linear%20Color%20Structure) (Tertiary :type Linear%20Color%20Structure) (AltPrimary :type Linear%20Color%20Structure) (AltSecondary :type Linear%20Color%20Structure) (AltTertiary :type Linear%20Color%20Structure) (MusicTimeMeasure :type Float%20%28single-precision%29))
  (if (not (Variables|Default|GetUseAlternateColorPalette))
    (Rendering|Components|Light|SetLightColor (Variables|Default|GetSpotLight) (Utilities|Select :Index (- (Variables|Default|GetColorSelector) 1) :Option%200 Primary :Option%201 Secondary :Option%202 Tertiary) false)
    (else
      (Rendering|Components|Light|SetLightColor (Variables|Default|GetSpotLight) (Utilities|Select :Index (- (Variables|Default|GetColorSelector) 1) :Option%200 AltPrimary :Option%201 AltSecondary :Option%202 AltTertiary) false)))
  (Rendering|SetVisibility (Variables|Default|GetSpotLight) true false))

(event EventStageShowActivateBeat
  (params (Active :type Float%20%28single-precision%29) (MusicTimeMeasure :type Float%20%28single-precision%29))
  (Rendering|SetVisibility (Variables|Default|GetSpotLight) true false))

(event EventStageShowActivate
  (params (Strength :type Float%20%28single-precision%29) (MusicTimeMeasure :type Float%20%28single-precision%29))
  (Rendering|SetVisibility (Variables|Default|GetSpotLight) (>= Strength 0.5) false))

(event EventStageShowImpulse
  (params (MusicTimeMeasure :type Float%20%28single-precision%29))
  (Rendering|SetVisibility (Variables|Default|GetSpotLight) true false))

(event EventStageShowUpdateNiagaraSystem
  (params (NiagaraSystem :type Niagara%20System%20Object%20Reference)))

(fn MojoStageShowActivate
  (params (Strength :type Float%20%28double-precision%29) (MusicTimeMeasure :type Float%20%28double-precision%29))
  (Rendering|SetVisibility (Variables|Default|GetSpotLight) (>= Strength 0.5) false)
  (return false))

(fn MojoStageShowActivateBeat
  (params (Active :type Float%20%28double-precision%29) (MusicTimeMeasure :type Float%20%28double-precision%29))
  (Rendering|SetVisibility (Variables|Default|GetSpotLight) true false)
  (return false))

(fn MojoStageShowColorShift
  (params (Primary :type Linear%20Color%20Structure) (Secondary :type Linear%20Color%20Structure) (Tertiary :type Linear%20Color%20Structure) (AltPrimary :type Linear%20Color%20Structure) (AltSecondary :type Linear%20Color%20Structure) (AltTertiary :type Linear%20Color%20Structure) (MusicTimeMeasure :type Float%20%28double-precision%29))
  (if (not (Variables|Default|GetUseAlternateColorPalette))
    (Rendering|Components|Light|SetLightColor (Variables|Default|GetSpotLight) (Utilities|Select :Index (- (Variables|Default|GetColorSelector) 1) :Option%200 Primary :Option%201 Secondary :Option%202 Tertiary) false)
    (else
      (Rendering|Components|Light|SetLightColor (Variables|Default|GetSpotLight) (Utilities|Select :Index (- (Variables|Default|GetColorSelector) 1) :Option%200 AltPrimary :Option%201 AltSecondary :Option%202 AltTertiary) false)))
  (Rendering|SetVisibility (Variables|Default|GetSpotLight) true false)
  (return true))

(fn MojoStageShowCustomMode
  (params (ModeIndex :type Integer) (MusicTimeMeasure :type Float%20%28double-precision%29))
  (CallFunction|StageShowCustomMode ModeIndex MusicTimeMeasure)
  (return false))

(fn MojoStageShowImpulse
  (params (MusicTimeMeasure :type Float%20%28double-precision%29))
  (Rendering|SetVisibility (Variables|Default|GetSpotLight) true false)
  (return false))

(fn MojoStageShowOnStop
  (CallFunction|StageShowonStop)
  (return false))

(fn MojoStageShowUpdateNiagaraSystem
  (params (Niagara%20System :type Niagara%20System%20Object%20Reference))
  (CallFunction|StageShowUpdateNiagaraSystem Niagara%20System)
  (return false))

(fn ConstructionScript
  (bind _spotlight (Variables|Default|GetSpotLight))
  (Rendering|Components|Light|SetIntensity _spotlight (Variables|Default|GetLightIntensity))
  (Rendering|Lighting|SetOuterConeAngle _spotlight (Variables|Default|GetOuterConeAngle))
  (Rendering|Lighting|SetAttenuationRadius _spotlight (Variables|Default|GetAttenuationRadius))
  (Rendering|Components|Light|SetCastShadows _spotlight (Variables|Default|GetShadowCasting)))
