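; Venue point light Blueprint that responds to music-timed stage show cues.
; ColorShift colors the light from one of three palette slots via Color Selector, using the alternate palette if flagged, and makes it visible.
; ActivateBeat and Impulse always show the light, while Activate shows it only at Strength 0.5 or higher.
; Mojo functions duplicate these handlers or forward CustomMode, OnStop and Niagara updates to StageShow functions, returning false.
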(event EventBeginPlay)

(event Collision|EventActorBeginOverlap
  (params (OtherActor :type Actor%20Object%20Reference)))

(event EventTick
  (params (DeltaSeconds :type Float%20%28single-precision%29)))

(event EventStageShowColorShift
  (params (Primary :type Linear%20Color%20Structure) (Secondary :type Linear%20Color%20Structure) (Tertiary :type Linear%20Color%20Structure) (AltPrimary :type Linear%20Color%20Structure) (AltSecondary :type Linear%20Color%20Structure) (AltTertiary :type Linear%20Color%20Structure) (MusicTimeMeasure :type Float%20%28single-precision%29))
  (if (not (Variables|Default|GetUseAlternateColorPalette))
    (Rendering|Components|Light|SetLightColor (Variables|Default|GetPointLight) (Utilities|Select :Index (Variables|Default|GetColor%20Selector) :Option%200 Primary :Option%201 Secondary :Option%202 Tertiary) false)
    (else
      (Rendering|Components|Light|SetLightColor (Variables|Default|GetPointLight) (Utilities|Select :Index (Variables|Default|GetColor%20Selector) :Option%200 AltPrimary :Option%201 AltSecondary :Option%202 AltTertiary) false)))
  (Rendering|SetVisibility (Variables|Default|GetPointLight) true false))

(event EventStageShowActivateBeat
  (params (Active :type Float%20%28single-precision%29) (MusicTimeMeasure :type Float%20%28single-precision%29))
  (Rendering|SetVisibility (Variables|Default|GetPointLight) true false))

(event EventStageShowActivate
  (params (Strength :type Float%20%28single-precision%29) (MusicTimeMeasure :type Float%20%28single-precision%29))
  (Rendering|SetVisibility (Variables|Default|GetPointLight) (>= Strength 0.5) false))

(event EventStageShowImpulse
  (params (MusicTimeMeasure :type Float%20%28single-precision%29))
  (Rendering|SetVisibility (Variables|Default|GetPointLight) true false))

(fn MojoStageShowActivate
  (params (Strength :type Float%20%28double-precision%29) (MusicTimeMeasure :type Float%20%28double-precision%29))
  (Rendering|SetVisibility (Variables|Default|GetPointLight) (>= Strength 0.5) false)
  (return false))

(fn MojoStageShowActivateBeat
  (params (Active :type Float%20%28double-precision%29) (MusicTimeMeasure :type Float%20%28double-precision%29))
  (Rendering|SetVisibility (Variables|Default|GetPointLight) true false)
  (return false))

(fn MojoStageShowColorShift
  (params (Primary :type Linear%20Color%20Structure) (Secondary :type Linear%20Color%20Structure) (Tertiary :type Linear%20Color%20Structure) (AltPrimary :type Linear%20Color%20Structure) (AltSecondary :type Linear%20Color%20Structure) (AltTertiary :type Linear%20Color%20Structure) (MusicTimeMeasure :type Float%20%28double-precision%29))
  (if (not (Variables|Default|GetUseAlternateColorPalette))
    (Rendering|Components|Light|SetLightColor (Variables|Default|GetPointLight) (Utilities|Select :Index (Variables|Default|GetColor%20Selector) :Option%200 Primary :Option%201 Secondary :Option%202 Tertiary) false)
    (else
      (Rendering|Components|Light|SetLightColor (Variables|Default|GetPointLight) (Utilities|Select :Index (Variables|Default|GetColor%20Selector) :Option%200 AltPrimary :Option%201 AltSecondary :Option%202 AltTertiary) false)))
  (Rendering|SetVisibility (Variables|Default|GetPointLight) true false)
  (return false))

(fn MojoStageShowCustomMode
  (params (ModeIndex :type Integer) (MusicTimeMeasure :type Float%20%28double-precision%29))
  (CallFunction|StageShowCustomMode ModeIndex MusicTimeMeasure)
  (return false))

(fn MojoStageShowImpulse
  (params (MusicTimeMeasure :type Float%20%28double-precision%29))
  (Rendering|SetVisibility (Variables|Default|GetPointLight) true false)
  (return false))

(fn MojoStageShowOnStop
  (CallFunction|StageShowonStop)
  (return false))

(fn MojoStageShowUpdateNiagaraSystem
  (params (Niagara%20System :type Niagara%20System%20Object%20Reference))
  (CallFunction|StageShowUpdateNiagaraSystem Niagara%20System)
  (return false))

(fn ConstructionScript)
