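; Venue spotlight Blueprint that swings back and forth or sweeps between start and end rotations and responds to stage show cues.
; On cosmetic-capable instances, a timeline lerps Cube_Rotator's rotation, phase-shifted by AnimationOffset.
; Activate cues above a small strength threshold restart the sweep and show the light; color shifts pick a palette color via ColorSelector.
; Mojo functions duplicate activation and color logic, forwarding other cues to StageShow calls.
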(event EventBeginPlay
  (Networking|CanExecuteCosmeticEvents
    (:True
      (Components|Timeline|SetNewTime (Variables|B_Managed_VenueSpotlight_Rotation|GetTimeline) (* (Variables|Default|GetAnimationOffset) (Components|Timeline|GetTimelineLength (Variables|B_Managed_VenueSpotlight_Rotation|GetTimeline))))
      (bind (_direction _rotation _rotlerp) (Timeline|Timeline Play)
        (:Update
          (Transformation|SetRelativeRotation (Variables|Default|GetCube_Rotator) (Math|Rotator|Lerp(Rotator) (Variables|Default|GetStartingRotation) (Variables|Default|GetEndRotation) _rotlerp) false false))))))

(event EventStageShowColorShift
  (params (Primary :type Linear%20Color%20Structure) (Secondary :type Linear%20Color%20Structure) (Tertiary :type Linear%20Color%20Structure) (AltPrimary :type Linear%20Color%20Structure) (AltSecondary :type Linear%20Color%20Structure) (AltTertiary :type Linear%20Color%20Structure) (MusicTimeMeasure :type Float%20%28single-precision%29))
  (Rendering|Components|Light|SetLightColor (Variables|Default|GetSpotLight) (Utilities|Select :Index (Variables|Default|GetColorSelector) :Option%200 Primary :Option%201 Secondary :Option%202 Tertiary) false))

(event EventStageShowActivate
  (params (Strength :type Float%20%28single-precision%29) (MusicTimeMeasure :type Float%20%28single-precision%29))
  (bind _returnvalue (> Strength 0.1))
  (bind _timeline (Variables|B_Managed_VenueSpotlight_Rotation|GetTimeline))
  (if _returnvalue
    (Components|Timeline|SetPlaybackPosition _timeline (* (Variables|Default|GetAnimationOffset) (Components|Timeline|GetTimelineLength _timeline)) false true)
    (Components|Timeline|Play _timeline)
    (else
      (Components|Timeline|Stop _timeline)))
  (Rendering|SetVisibility (Variables|Default|GetSpotLight) _returnvalue false))

(fn MojoStageShowActivate
  (params (Strength :type Float%20%28double-precision%29) (MusicTimeMeasure :type Float%20%28double-precision%29))
  (bind _returnvalue (> Strength 0.1))
  (bind _timeline (Variables|B_Managed_VenueSpotlight_Rotation|GetTimeline))
  (if _returnvalue
    (Components|Timeline|SetPlaybackPosition _timeline (* (Variables|Default|GetAnimationOffset) (Components|Timeline|GetTimelineLength _timeline)) false true)
    (Components|Timeline|Play _timeline)
    (else
      (Components|Timeline|Stop _timeline)))
  (Rendering|SetVisibility (Variables|Default|GetSpotLight) _returnvalue false)
  (return false))

(fn MojoStageShowActivateBeat
  (params (Active :type Float%20%28double-precision%29) (MusicTimeMeasure :type Float%20%28double-precision%29))
  (CallFunction|StageShowActivateBeat Active MusicTimeMeasure)
  (return false))

(fn MojoStageShowColorShift
  (params (Primary :type Linear%20Color%20Structure) (Secondary :type Linear%20Color%20Structure) (Tertiary :type Linear%20Color%20Structure) (AltPrimary :type Linear%20Color%20Structure) (AltSecondary :type Linear%20Color%20Structure) (AltTertiary :type Linear%20Color%20Structure) (MusicTimeMeasure :type Float%20%28double-precision%29))
  (Rendering|Components|Light|SetLightColor (Variables|Default|GetSpotLight) (Utilities|Select :Index (Variables|Default|GetColorSelector) :Option%200 Primary :Option%201 Secondary :Option%202 Tertiary) false)
  (return false))

(fn MojoStageShowCustomMode
  (params (ModeIndex :type Integer) (MusicTimeMeasure :type Float%20%28double-precision%29))
  (CallFunction|StageShowCustomMode ModeIndex MusicTimeMeasure))

(fn MojoStageShowImpulse
  (params (MusicTimeMeasure :type Float%20%28double-precision%29))
  (CallFunction|StageShowImpulse MusicTimeMeasure)
  (return false))

(fn MojoStageShowOnStop
  (CallFunction|StageShowonStop)
  (return false))

(fn MojoStageShowUpdateNiagaraSystem
  (params (Niagara%20System :type Niagara%20System%20Object%20Reference))
  (CallFunction|StageShowUpdateNiagaraSystem Niagara%20System)
  (return false))

(fn ConstructionScript)
