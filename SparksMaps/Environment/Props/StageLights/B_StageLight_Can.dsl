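; B_StageLight_Can is a stage light can prop Blueprint whose behavior comes from its parent class.
; It overrides MojoStageShow hooks for activate, beat, color shift, custom mode, impulse, stop and Niagara system update, with most receiving a music time measure.
; Each override and the construction script only forwards to the parent, so this graph adds no logic of its own.
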(fn MojoStageShowActivate
  (params (Strength :type Float%20%28double-precision%29) (MusicTimeMeasure :type Float%20%28double-precision%29))
  (bind _result (|CallParentFunction Strength MusicTimeMeasure))
  (return _result))

(fn MojoStageShowActivateBeat
  (params (Active :type Float%20%28double-precision%29) (MusicTimeMeasure :type Float%20%28double-precision%29))
  (bind _result (|CallParentFunction Active MusicTimeMeasure))
  (return _result))

(fn MojoStageShowColorShift
  (params (Primary :type Linear%20Color%20Structure) (Secondary :type Linear%20Color%20Structure) (Tertiary :type Linear%20Color%20Structure) (AltPrimary :type Linear%20Color%20Structure) (AltSecondary :type Linear%20Color%20Structure) (AltTertiary :type Linear%20Color%20Structure) (MusicTimeMeasure :type Float%20%28double-precision%29))
  (bind _result (|CallParentFunction Primary Secondary Tertiary AltPrimary AltSecondary AltTertiary MusicTimeMeasure))
  (return _result))

(fn MojoStageShowCustomMode
  (params (ModeIndex :type Integer) (MusicTimeMeasure :type Float%20%28double-precision%29))
  (bind _result (|CallParentFunction ModeIndex MusicTimeMeasure))
  (return _result))

(fn MojoStageShowImpulse
  (params (MusicTimeMeasure :type Float%20%28double-precision%29))
  (bind _result (|CallParentFunction MusicTimeMeasure))
  (return _result))

(fn MojoStageShowOnStop
  (bind _result (|CallParentFunction))
  (return _result))

(fn MojoStageShowUpdateNiagaraSystem
  (params (Niagara%20System :type Niagara%20System%20Object%20Reference))
  (bind _result (|CallParentFunction Niagara%20System))
  (return _result))

(fn ConstructionScript
  (|CallParentFunction))
