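; Base Blueprint for a can-style stage light prop that responds to Mojo stage show events.
; Mojo functions for activation, beat activation, color shift, custom mode, impulse, stop and Niagara system updates each forward their inputs to a matching StageShow function and return false.
; Most events carry a music time measure, and color shift passes primary, secondary and tertiary colors plus alternates.
; The construction script only calls its parent.
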(fn MojoStageShowActivate
  (params (Strength :type Float%20%28double-precision%29) (MusicTimeMeasure :type Float%20%28double-precision%29))
  (CallFunction|StageShowActivate Strength MusicTimeMeasure)
  (return false))

(fn MojoStageShowActivateBeat
  (params (Active :type Float%20%28double-precision%29) (MusicTimeMeasure :type Float%20%28double-precision%29))
  (CallFunction|StageShowActivateBeat Active MusicTimeMeasure)
  (return false))

(fn MojoStageShowColorShift
  (params (Primary :type Linear%20Color%20Structure) (Secondary :type Linear%20Color%20Structure) (Tertiary :type Linear%20Color%20Structure) (AltPrimary :type Linear%20Color%20Structure) (AltSecondary :type Linear%20Color%20Structure) (AltTertiary :type Linear%20Color%20Structure) (MusicTimeMeasure :type Float%20%28double-precision%29))
  (CallFunction|StageShowColorShift Primary Secondary Tertiary AltPrimary AltSecondary AltTertiary MusicTimeMeasure)
  (return false))

(fn MojoStageShowCustomMode
  (params (ModeIndex :type Integer) (MusicTimeMeasure :type Float%20%28double-precision%29))
  (CallFunction|StageShowCustomMode ModeIndex MusicTimeMeasure)
  (return false))

(fn MojoStageShowImpulse
  (params (MusicTimeMeasure :type Float%20%28double-precision%29))
  (CallFunction|StageShowImpulse MusicTimeMeasure)
  (return false))

(fn MojoStageShowOnStop
  (CallFunction|StageShowonStop)
  (return false))

(fn MojoStageShowUpdateNiagaraSystem
  (params (Niagara%20System :type Niagara%20System%20Object%20Reference))
  (CallFunction|StageShowUpdateNiagaraSystem Niagara%20System)
  (return false))

(fn ConstructionScript
  (|CallParentFunction))
