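; B_StageLight_LightshaftOnly is a stage light Blueprint that drives a spotlight beam mesh in music-synced stage shows.
; Stage show events and matching Mojo functions write activation strength, beat pulse, music timing and impulse time into the mesh's Custom Primitive Data for its material.
; Beat and impulse calls show the beam, while activation sets visibility after a retriggerable delay, hiding low-strength beams.
; Construction seeds static color and full strength and applies an optional override material.
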(event Custom|DisableBeamDelay
  (params (Duration :type Float%20%28double-precision%29) (Enabled :type Boolean))
  (Utilities|FlowControl|RetriggerableDelay Duration)
  (Rendering|SetVisibility (Variables|Default|GetSM_Sparks_Spotlight) Enabled false))

(event EventStageShowActivateBeat
  (params (Active :type Float%20%28single-precision%29) (MusicTimeMeasure :type Float%20%28single-precision%29))
  (bind _sm_sparks_spotlight (Variables|Default|GetSM_Sparks_Spotlight))
  (Rendering|Material|SetScalarParameterforCustomPrimitiveData _sm_sparks_spotlight "BeatPulse_Active" Active)
  (Rendering|Material|SetScalarParameterforCustomPrimitiveData _sm_sparks_spotlight "StageShow_MusicTimeSignature" MusicTimeMeasure)
  (Rendering|SetVisibility (Variables|Default|GetSM_Sparks_Spotlight) true false))

(event EventStageShowActivate
  (params (Strength :type Float%20%28single-precision%29) (MusicTimeMeasure :type Float%20%28single-precision%29))
  (bind _sm_sparks_spotlight (Variables|Default|GetSM_Sparks_Spotlight))
  (Rendering|Material|SetScalarParameterforCustomPrimitiveData _sm_sparks_spotlight "StageShow_ActivateStrength" Strength)
  (Rendering|Material|SetScalarParameterforCustomPrimitiveData _sm_sparks_spotlight "StageShow_MusicTimeSignature" MusicTimeMeasure)
  (CallFunction|DisableBeamDelay 0.5 (>= Strength 0.5)))

(event EventStageShowImpulse
  (params (MusicTimeMeasure :type Float%20%28single-precision%29))
  (bind _sm_sparks_spotlight (Variables|Default|GetSM_Sparks_Spotlight))
  (Rendering|Material|SetScalarParameterforCustomPrimitiveData _sm_sparks_spotlight "StageShow_MusicTimeSignature" (Math|Float|SelectFloat 3.0 0.5 (> MusicTimeMeasure 0.6)))
  (Rendering|Material|SetScalarParameterforCustomPrimitiveData _sm_sparks_spotlight "StageShow_ImpulseTime" (Utilities|Time|GetGameTimeinSeconds))
  (Rendering|SetVisibility _sm_sparks_spotlight true false))

(fn MojoStageShowActivate
  (params (Strength :type Float%20%28double-precision%29) (MusicTimeMeasure :type Float%20%28double-precision%29))
  (bind _sm_sparks_spotlight (Variables|Default|GetSM_Sparks_Spotlight))
  (Rendering|Material|SetScalarParameterforCustomPrimitiveData _sm_sparks_spotlight "StageShow_ActivateStrength" Strength)
  (Rendering|Material|SetScalarParameterforCustomPrimitiveData _sm_sparks_spotlight "StageShow_MusicTimeSignature" MusicTimeMeasure)
  (CallFunction|DisableBeamDelay 0.5 (>= Strength 0.5))
  (return false))

(fn MojoStageShowActivateBeat
  (params (Active :type Float%20%28double-precision%29) (MusicTimeMeasure :type Float%20%28double-precision%29))
  (bind _sm_sparks_spotlight (Variables|Default|GetSM_Sparks_Spotlight))
  (Rendering|Material|SetScalarParameterforCustomPrimitiveData _sm_sparks_spotlight "BeatPulse_Active" Active)
  (Rendering|Material|SetScalarParameterforCustomPrimitiveData _sm_sparks_spotlight "StageShow_MusicTimeSignature" MusicTimeMeasure)
  (Rendering|SetVisibility (Variables|Default|GetSM_Sparks_Spotlight) true false)
  (return false))

(fn MojoStageShowColorShift
  (params (Primary :type Linear%20Color%20Structure) (Secondary :type Linear%20Color%20Structure) (Tertiary :type Linear%20Color%20Structure) (AltPrimary :type Linear%20Color%20Structure) (AltSecondary :type Linear%20Color%20Structure) (AltTertiary :type Linear%20Color%20Structure) (MusicTimeMeasure :type Float%20%28double-precision%29))
  (CallFunction|StageShowColorShift Primary Secondary Tertiary AltPrimary AltSecondary AltTertiary MusicTimeMeasure)
  (return false))

(fn MojoStageShowCustomMode
  (params (ModeIndex :type Integer) (MusicTimeMeasure :type Float%20%28double-precision%29))
  (CallFunction|StageShowCustomMode ModeIndex MusicTimeMeasure)
  (return false))

(fn MojoStageShowImpulse
  (params (MusicTimeMeasure :type Float%20%28double-precision%29))
  (bind _sm_sparks_spotlight (Variables|Default|GetSM_Sparks_Spotlight))
  (Rendering|Material|SetScalarParameterforCustomPrimitiveData _sm_sparks_spotlight "StageShow_MusicTimeSignature" (Math|Float|SelectFloat 3.0 0.5 (> MusicTimeMeasure 0.6)))
  (Rendering|Material|SetScalarParameterforCustomPrimitiveData _sm_sparks_spotlight "StageShow_ImpulseTime" (Utilities|Time|GetGameTimeinSeconds))
  (Rendering|SetVisibility _sm_sparks_spotlight true false)
  (return false))

(fn MojoStageShowOnStop
  (CallFunction|StageShowonStop)
  (return false))

(fn MojoStageShowUpdateNiagaraSystem
  (params (Niagara%20System :type Niagara%20System%20Object%20Reference))
  (CallFunction|StageShowUpdateNiagaraSystem Niagara%20System)
  (return false))

(fn OverrideMaterial
  (bind _lightbeamoverridematerial (Variables|Default|GetLightbeamOverrideMaterial)
    (:then
      (Variables|Default|SetLightbeamOriginalMaterial (Rendering|Material|GetMaterial (Variables|Default|GetSM_Sparks_Spotlight) 0))
      (Rendering|Material|SetMaterial (Variables|Default|GetSM_Sparks_Spotlight) 0 _lightbeamoverridematerial))
    (:else
      (bind _lightbeamoriginalmaterial (Variables|Default|GetLightbeamOriginalMaterial)
        (:then
          (Rendering|Material|SetMaterial (Variables|Default|GetSM_Sparks_Spotlight) 0 _lightbeamoriginalmaterial))
        (:else)))))

(fn ConstructionScript
  (bind (_staticcolor_r _staticcolor_g _staticcolor_b _staticcolor_a) (Math|Color|BreakColor (Variables|GetStaticColor_R)))
  (Rendering|Material|SetScalarParameterforDefaultCustomPrimitiveData (Variables|Default|GetSM_Sparks_Spotlight) "StageShow_ActivateStrength" 1.0)
  (CallFunction|OverrideMaterial)
  (Rendering|Material|SetVectorParameterforDefaultCustomPrimitiveData (Variables|Default|GetSM_Sparks_Spotlight) "Color" (Math|Vector4|MakeVector4 _staticcolor_r _staticcolor_g _staticcolor_b _staticcolor_a)))
