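# Material function that outputs a "Stage Show Anim" value for stage show materials, combining a base emissive level with a timed pulse.
# The base multiplies emissive strength by an activation strength read from custom primitive data.
# Time since an impulse timestamp drives a linear ramp-in and fade-out envelope, scaled by pulse strength plus music time signature.
# Impulse time and time signature also come from custom primitive data, allowing per-primitive triggering.
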MaterialFunction /SparksStageShow/Systems/StageManager/FXManager/MaterialFunctions/MF_MojoStageShow
graph {nodes [
  n1 = ScalarParameter() {defaultValue=1, group=StageShow, parameterName=StageShow_EmissiveStrength, primitiveDataIndex=4}
  n2 = ScalarParameter() {bUseCustomPrimitiveData=true, defaultValue=1, group=StageShow, parameterName=StageShow_ActivateStrength, primitiveDataIndex=4}
  n3 = Add(
   A=ScalarParameter() {defaultValue=1, group=StageShow, parameterName=StageShow_PulseStrength}
   B=ScalarParameter() {bUseCustomPrimitiveData=true, parameterName=StageShow_MusicTimeSignature, primitiveDataIndex=5}
  )
  n4 = Subtract(A=Time(), B=ScalarParameter() {bUseCustomPrimitiveData=true, group=StageShow, parameterName=StageShow_ImpulseTime, primitiveDataIndex=6})
  n5 = ScalarParameter() {defaultValue=0.1, group=StageShow, parameterName=StageShow_ImpulseLengthIn}
  n6 = Divide(A=@n4, B=Add(A=@n5, B=ScalarParameter() {defaultValue=0.5, group=StageShow, parameterName=StageShow_ImpulseLengthOut}))
  n7 = Multiply(A=Saturate(None=Divide(A=@n4, B=@n5) {constB=0.4}), B=OneMinus(None=Saturate(None=@n6))) {constB=2}
  FunctionOutput(None=Add(A=Multiply(A=@n1, B=@n2), B=Multiply(A=@n3, B=@n7))) {outputName=Stage Show Anim}
  ScalarParameter() {defaultValue=1, group=StageShow, parameterName=StageShow_PulseStrength}
  ScalarParameter() {parameterName=StageShow_MusicTimeSignature}
 ]}
