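# Additive, unlit, two-sided material for Sparks stage spotlight beam shafts.
# Color comes from stage-show colors or custom primitive data, with switchable show animation and beat pulse.
# SplineThicken shapes and stretches the beam, collapsing it when show animation is zero.
# UV-built glow fades by camera distance and view angle, plus depth except on mobile and low quality; smoke, noise and rotation branches are also present.
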Material /SparksMaps/Environment/Props/StageLights/Materials/M_Sparks_Spotlight
graph {
 properties {
  bUsedWithInstancedStaticMeshes=true
  blendMode=BLEND_Additive
  pixelDepthOffsetMode=PDOM_Legacy
  shadingModel=MSM_Unlit
  twoSided=true
  usedShadingModels=Unlit
 }
 nodes [
  n1 = Constant3Vector() {constant=<rgba 0 1 0 1>}
  n2 = Frac(None=Multiply(A=Time()) {constB=0.5})
  n3 = Constant() {r=0.15}
  n4 = Transform(None=Constant3Vector() {constant=<rgba 70 0 0 1>}) {transformSourceType=TRANSFORMSOURCE_Local}
  n5 = ActorPositionWS()
  n6 = WorldPosition()
  Add(B=RotateAboutAxis(
    NormalizedRotationAxis=Transform(None=@n1) {transformSourceType=TRANSFORMSOURCE_Local}
    RotationAngle=Multiply(A=@n2, B=@n3) {constB=0.15}
    PivotPoint=Add(A=@n4, B=@n5)
    Position=Add(B=@n6.XYZ)
   ))
  n7 = MaterialFunctionCall() {materialFunction=/SparksStageShow/Systems/StageManager/FXManager/MaterialFunctions/MF_MojoStageShow}
  n8 = Time()
  n9 = Add(A=MaterialFunctionCall!(CustomUPosition=Frac(None=@n8)) {materialFunction=/SparksStageShow/Systems/StageManager/FXManager/MaterialFunctions/MF_MojoStageColors}.Primary) {constB=0.1}
  n10 = StaticSwitchParameter(
   True=Multiply(A=Saturate(None=@n9)) {constB=2}
   False=Multiply(A=VectorParameter() {bUseCustomPrimitiveData=true, parameterName=Color, defaultValue=<rgba 0.5633 0.0833 1 1>}.RGB) {constB=2}
  ) {defaultValue=true, parameterName=UseStageShowColor}
  n11 = StaticSwitchParameter(
   True=MaterialFunctionCall!(
    Emissive=@n10
    BeatPulse Strength=ScalarParameter() {defaultValue=2, parameterName=BeatPulse_Strength}
    BeatPulse Min=Constant() {r=0.25}
    BeatPulse Max
   ) {materialFunction=/SparksMaps/Environment/MaterialFunctions/MF_MojoBeatPulse}.Result
   False=@n10
  ) {parameterName=UseMojoBeatPulse}
  MaterialFunctionCall(Emissive=StaticSwitchParameter(True=Multiply(A=@n7.Stage Show Anim, B=@n11), False=@n11) {defaultValue=true, parameterName=UseMojoStageShow}) {materialFunction=&MF_FX_EmissiveAdjustments}
  n12 = TextureCoordinate()
  n13 = ScalarParameter() {defaultValue=1, parameterName=Glow Tip Scale}
  n14 = ScalarParameter() {defaultValue=1, parameterName=Glow Base Scale}
  n15 = ComponentMask(None=@n12) {g=true}
  n16 = ComponentMask(None=MaterialFunctionCall(UVs=@n12, Texture Scale=LinearInterpolate(A=@n13, B=@n14, Alpha=@n15)) {materialFunction=&ScaleUVsByCenter}.UVs) {r=true}
  n17 = Multiply(A=@n16, B=OneMinus(None=@n16))
  n18 = ScalarParameter() {defaultValue=3, parameterName=Glow Edge Fade Power}
  n19 = Saturate(None=Multiply(A=Power(Base=@n17, Exp=@n18) {constExponent=4}, B=ScalarParameter() {defaultValue=5, parameterName=Glow Center Column Multiplier}))
  n20 = ComponentMask(None=TextureCoordinate()) {g=true}
  n21 = Power(Base=@n20, Exp=ScalarParameter() {defaultValue=1, parameterName=Glow Tip Fade Power})
  n22 = Multiply(
   A=Power(Base=@n20, Exp=ScalarParameter() {defaultValue=10, parameterName=Glow Tip Fade Power2})
   B=ScalarParameter() {defaultValue=10, parameterName=Glow Tip Fade Power2 Brightness}
  )
  n23 = Multiply(A=Add(A=Multiply(A=@n19, B=@n21), B=Multiply(A=@n19, B=@n22)), B=ScalarParameter() {defaultValue=0.5, parameterName=Glow Brightness Mult})
  n24 = Multiply(A=@n23, B=DepthFade())
  n25 = QualitySwitch(Default=@n24, Low=@n23, High=@n24, Medium=@n24)
  n26 = ScalarParameter() {
   defaultValue=1
   desc=Default HDRMult is too bright for mobile. This allows to reduce for mobile when HDR is high.
   parameterName=MobileBrightnessMultiplier
  }
  n27 = Multiply(A=@n26, B=@n23)
  n28 = ScalarParameter() {defaultValue=500, parameterName=NearFade}
  n29 = MaterialFunctionCall(
   WidthBase=ScalarParameter() {defaultValue=40, parameterName=Mesh Base Width}
   WidthTip=ScalarParameter() {defaultValue=400, parameterName=Mesh Tip Width}
  ) {desc=Can't be done in instances, materialFunction=&SplineThicken}
  n30 = Abs(None=DotProduct(A=CameraVectorWS(), B=@n29.Normal))
  n31 = ScalarParameter() {defaultValue=0.5, parameterName=ViewAngleFade}
  Multiply(
   A=Multiply(
    A=ShadingPathSwitch(Default=@n25, Mobile=@n27)
    B=MaterialFunctionCall(Distance=@n28) {materialFunction=&Near_Camera_Fade}.Result
   )
   B=Clamp(None=Power(Base=@n30, Exp=@n31))
  )
  n32 = TextureCoordinate()
  n33 = Multiply(A=OneMinus(None=ComponentMask(None=@n32) {g=true}), B=ScalarParameter() {defaultValue=2500, parameterName=Mesh Scale Length})
  n34 = Transform(None=Constant3Vector() {constant=<rgba 1 0 0 0>}) {transformSourceType=TRANSFORMSOURCE_Local}
  Multiply(A=Saturate(None=Ceil(None=@n7.Stage Show Anim)), B=Add(A=@n29.WorldPositionOffset, B=Multiply(A=@n33, B=@n34)))
  n35 = ActorPositionWS()
  n36 = ObjectPositionWS()
  n37 = Add(A=Add(A=ComponentMask(None=@n36) {r=true}, B=ComponentMask(None=@n36) {g=true}), B=ComponentMask(None=@n36) {b=true})
  n38 = Panner(
   Coordinate=Multiply(
    A=MaterialFunctionCall!(Center Point World Position=@n35) {materialFunction=/Engine/Functions/Engine_MaterialFunctions02/UVs/WorldSpaceAlignedScreenCoordinates}.Offset and Scaled UVS
    B=ScalarParameter() {defaultValue=3.5, parameterName=Smoke Tile}
   )
   Time=Add(A=Time(), B=Divide(A=@n37) {constB=100})
  ) {speedY=0.1}
  n39 = Constant()
  n40 = Constant() {r=1}
  n41 = MaterialFunctionCall() {materialFunction=&LinearGradient}
  n42 = Multiply(
   A=ScalarParameter() {defaultValue=0.5, parameterName=Smoke Mult}
   B=Multiply(
    A=TextureSample(UVs=@n38) {texture=/Game/Global/Textures/Default/Tileable_Patterns/T_Blowing_Smoke_Pattern2_Tile_DISP}.R
    B=SmoothStep(Min=@n39, Max=@n40, Value=@n41.VGradient)
   )
  )
  n43 = Constant() {r=0.25}
  Multiply(A=@n23, B=Saturate(None=Add(A=@n42, B=@n43)))
  n44 = WorldPosition()
  n45 = Multiply(A=Constant() {r=0.0002}, B=Constant() {r=15})
  n46 = TextureSample!(UVs=ComponentMask(None=Multiply(A=@n44.XYZ, B=@n45)) {g=true, r=true}, Tex, Bias, Apply View MipBias) {
   mipValueMode=TMVM_MipBias
   samplerSource=SSM_Wrap_WorldGroupSettings
   samplerType=SAMPLERTYPE_LinearGrayscale
   texture=/Game/Effects/Fort_Effects/Emotes/UltraLight/Texture/T_Tiling_Noise_01_GreyScale
  }
  n47 = WorldPosition()
  n48 = Multiply(A=@n45) {constB=1.185}
  n49 = TextureSample(UVs=ComponentMask(None=Multiply(A=@n47.XYZ, B=@n48)) {b=true, r=true}) {
   mipValueMode=TMVM_MipBias
   samplerSource=SSM_Wrap_WorldGroupSettings
   samplerType=SAMPLERTYPE_LinearGrayscale
   texture=&T_Tiling_Noise_01_GreyScale
  }
  Multiply(A=@n46.R, B=@n49.R)
 ]
}
