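# Additive, unlit, two-sided material for stage light beams that react to music in the Sparks stage show.
# Color blends between stage palette and per-primitive custom color.
# Beam intensity multiplies two scrolling world-space noise layers, a light-pattern mask, inverted Fresnel edge softening and vertex-color fades, then feeds beat-pulse and show-animation functions.
# The graph also holds beat-gating, hue-shift and normal-offset branches, with material output connections not shown.
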Material /SparksMaps/Environment/Props/StageLights/Materials/M_StageLight
graph {
 properties {
  bUsedWithMeshDeformer=true
  bUsedWithSkeletalMesh=true
  blendMode=BLEND_Additive
  numCustomizedUVs=2
  pixelDepthOffsetMode=PDOM_Legacy
  shadingModel=MSM_Unlit
  twoSided=true
  usedShadingModels=Unlit
 }
 nodes [
  n1 = TextureCoordinate() {coordinateIndex=1}
  n2 = Constant() {r=1}
  n3 = Constant() {desc=Rows, r=512}
  AppendVector(A=ComponentMask(None=@n1) {r=true}, B=Multiply(A=ComponentMask(None=@n1) {g=true}, B=Divide(A=@n2, B=@n3)))
  Fmod(A=Multiply(A=Time()) {constB=3}, B=Constant() {r=4})
  n4 = CollectionParameter() {collection=/FMAudioReactiveData/AudioAnalysis/MPC_Sparks_Audio_Parameters, parameterName=BeatLifetime}
  n5 = Constant()
  n6 = ScalarParameter() {bUseCustomPrimitiveData=true, parameterName=bUseCustomColor, primitiveDataIndex=27}
  n7 = VectorParameter() {bUseCustomPrimitiveData=true, parameterName=CustomColor, primitiveDataIndex=28, defaultValue=<rgba 1 0 0 0>}
  n8 = Frac(None=Time())
  n9 = Add(
   A=Multiply(A=@n6, B=@n7.RGB)
   B=Multiply(
    A=OneMinus(None=@n6)
    B=MaterialFunctionCall!(CustomUPosition=@n8) {materialFunction=/SparksStageShow/Systems/StageManager/FXManager/MaterialFunctions/MF_MojoStageColors}.Primary
   )
  )
  n10 = Constant() {r=1024}
  n11 = Divide(A=WorldPosition() {worldPositionShaderOffset=WPT_ExcludeAllShaderOffsets}.XYZ, B=Multiply(A=Abs(None=@n10)) {constB=-1})
  n12 = ComponentMask(None=@n11) {r=true}
  n13 = ComponentMask(None=@n11) {g=true}
  n14 = Time()
  n15 = Multiply(A=@n14) {constB=0.1}
  n16 = Multiply(A=@n14) {constB=0.09}
  n17 = TextureSample(UVs=Add(A=AppendVector(A=@n12, B=@n13), B=AppendVector(A=@n15, B=@n16))) {samplerSource=SSM_Wrap_WorldGroupSettings, texture=/Game/Global/Textures/Default/Noise/T_Tiling_Noise_01}
  n18 = Constant() {r=1460}
  n19 = Divide(A=WorldPosition() {worldPositionShaderOffset=WPT_ExcludeAllShaderOffsets}.XYZ, B=Multiply(A=Abs(None=@n18)) {constB=-1})
  n20 = ComponentMask(None=@n19) {r=true}
  n21 = ComponentMask(None=@n19) {g=true}
  n22 = Time()
  n23 = Multiply(A=@n22) {constB=-0.13}
  n24 = Multiply(A=@n22) {constB=0.127}
  n25 = TextureSample(UVs=Add(A=AppendVector(A=@n20, B=@n21), B=AppendVector(A=@n23, B=@n24))) {samplerSource=SSM_Wrap_WorldGroupSettings, texture=&T_Tiling_Noise_01}
  n26 = VertexColor()
  n27 = Multiply(A=@n26.A, B=@n26.A)
  n28 = TextureCoordinate() {coordinateIndex=1}
  n29 = Multiply(
   A=Add(A=Multiply(A=@n17.RGB, B=@n25.RGB)) {constB=0.02}
   B=Multiply(A=@n27, B=TextureSample(UVs=@n28) {texture=/Game/Athena/Environments/Festivus/MotorizedLights/T_Spot_x4_LightPatterns_01_M}.A)
  )
  n30 = Fresnel(ExponentIn=Constant() {r=2}, BaseReflectFractionIn=Constant() {r=0.8}, Normal=VertexNormalWS())
  n31 = Multiply(A=Power(Base=OneMinus(None=@n30))) {constB=3.8}
  n32 = Multiply(A=Multiply(A=Multiply(A=@n9, B=@n29), B=Multiply(A=@n26.A, B=@n31)), B=OneMinus(None=VertexColor().R))
  n33 = If(A=@n4, B=ScalarParameter() {defaultValue=4, parameterName=BeatOver}, "A > B"=@n5, "A == B"=@n5, "A < B"=@n32) {constB=2}
  If(A=@n4, B=ScalarParameter() {defaultValue=3, parameterName=BeatUnder}, "A > B"=@n33, "A == B"=@n33, "A < B"=Constant()) {constB=2}
  n34 = Multiply(A=OneMinus(None=@n4), B=@n32)
  If(
   A=CollectionParameter() {collection=&MPC_Sparks_Audio_Parameters, parameterName=MeasureLifetime}
   "A > B"=@n34
   "A < B"=Multiply(A=Constant() {r=5}, B=@n34)
  ) {constB=0.1}
  n35 = Time()
  MaterialFunctionCall(
   Hue Shift Percentage=Frac(None=Multiply(A=@n35) {constB=0.5})
   Texture=VectorParameter() {parameterName=Color, defaultValue=<rgba 1 0 0 0>}.RGB
  ) {materialFunction=&HueShift}
  Multiply(
   A=MaterialFunctionCall() {materialFunction=/SparksStageShow/Systems/StageManager/FXManager/MaterialFunctions/MF_MojoStageShow}.Stage Show Anim
   B=MaterialFunctionCall!(Emissive=@n32, BeatPulse Strength, BeatPulse Min, BeatPulse Max) {materialFunction=/SparksMaps/Environment/MaterialFunctions/MF_MojoBeatPulse}.Result
  )
  n36 = Transform(None=VertexNormalWS()) {transformSourceType=TRANSFORMSOURCE_World, transformType=TRANSFORM_Tangent}
  n37 = AppendVector(A=ComponentMask(None=@n36) {r=true}, B=ComponentMask(None=@n36) {g=true})
  n38 = Add(A=ComponentMask(None=@n36) {b=true}, B=Constant() {r=10}) {constB=0}
  Multiply(A=Transform(None=AppendVector(A=@n37, B=@n38)), B=OneMinus(None=@n27))
  ScalarParameter() {defaultValue=0.5, parameterName=EmisiveStrength}
 ]
}
