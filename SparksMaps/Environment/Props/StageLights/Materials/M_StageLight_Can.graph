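# Stage light can material for the SparksMaps props, with motorized spotlight textures and a glowing lens driven by the stage show.
# A reflection-vector parallax offset gives the lens mask depth.
# Mojo stage colors or a custom color tint it, with switches for beat pulse and show animation.
# Optional WPO rotates vertices about the instance Y axis, using time, custom primitive data and a hashed offset.
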Material /SparksMaps/Environment/Props/StageLights/Materials/M_StageLight_Can
graph {
 properties {
  bUsedWithMeshDeformer=true
  bUsedWithNanite=true
  bUsedWithSkeletalMesh=true
  bUsedWithStaticLighting=true
  numCustomizedUVs=2
  pixelDepthOffsetMode=PDOM_Legacy
  usedShadingModels=Default Lit
 }
 nodes [
  n1 = Time()
  n2 = ScalarParameter() {bUseCustomPrimitiveData=true, parameterName=TimeStamp, primitiveDataIndex=18}
  n3 = MaterialFunctionCall(Float3=NamedRerouteUsage()) {materialFunction=&BreakOutFloat3Components}
  n4 = Frac(None=Add(A=Add(A=@n3.R, B=@n3.R), B=@n3.B))
  n5 = Multiply(
   A=Add(A=Subtract(A=@n1, B=@n2), B=Multiply(A=@n4) {constB=33.3})
   B=ScalarParameter() {bUseCustomPrimitiveData=true, defaultValue=0.1, parameterName=WPOAnim_Speed, primitiveDataIndex=17}
  )
  n6 = Frac(None=Multiply(A=Sine(None=@n5), B=ScalarParameter() {bUseCustomPrimitiveData=true, defaultValue=0.1, parameterName=WPOAnim_MaxAngle, primitiveDataIndex=7}))
  NamedRerouteDeclaration(None=@n6) {name=Anim Rotation, nodeColor=<rgba 0 1 0.5078 1>}
  n7 = Normalize(VectorInput=Transform(None=Constant3Vector() {constant=<rgba 0 1 0 0>}) {transformSourceType=TRANSFORMSOURCE_Instance})
  n8 = VertexInterpolator(VS=@n7)
  NamedRerouteDeclaration(None=NamedRerouteDeclaration(None=ShaderStageSwitch(PixelShader=@n8.PS, VertexShader=@n7)) {name=Cone Y BeforeRotation, nodeColor=<rgba 1 0.6797 0 1>}) {name=RotationAxis, nodeColor=<rgba 0.0312 1 0 1>}
  OneMinus(None=@n6)
  n9 = MaterialFunctionCall!(CustomUPosition=Frac(None=Time())) {materialFunction=/SparksStageShow/Systems/StageManager/FXManager/MaterialFunctions/MF_MojoStageColors}
  n10 = TextureSample() {samplerType=SAMPLERTYPE_Normal, texture=/Game/Athena/Environments/Festivus/MotorizedLights/T_MotorizedSpotLight_01_N}
  n11 = MaterialFunctionCall!(
   Normal=MaterialFunctionCall(R=Multiply(A=@n10.R) {constB=0.1}, G=Multiply(A=@n10.G) {constB=0.1}, B=Constant() {r=1}) {materialFunction=&AppendMany}.RGB
   CameraVector=Transform(None=CameraVectorWS()) {transformSourceType=TRANSFORMSOURCE_World, transformType=TRANSFORM_Tangent}
  ) {materialFunction=/Engine/Functions/Engine_MaterialFunctions02/Utility/CustomReflectionVector}
  n12 = ComponentMask(None=@n11.Reflection Vector) {g=true, r=true}
  n13 = Divide(A=ScalarParameter() {defaultValue=35, parameterName=Depth}, B=Abs(None=ComponentMask(None=@n11.Reflection Vector) {b=true}))
  n14 = Constant() {r=1}
  n15 = Constant() {r=1024}
  n16 = Add(
   A=Multiply(A=ScalarParameter() {defaultValue=1, parameterName=UVTile_Depth}, B=TextureCoordinate())
   B=Multiply(A=Multiply(A=@n12, B=@n13), B=Divide(A=@n14, B=@n15))
  )
  n17 = Multiply(
   A=StaticSwitchParameter(
    True=VectorParameter() {bUseCustomPrimitiveData=true, parameterName=Color, defaultValue=<rgba 0.0752 0.3268 0.5104 1>}.RGB
    False=Multiply(A=@n9.Primary) {constB=3}
   ) {parameterName=UseCustomColor}
   B=Multiply(
    A=TextureSample(UVs=@n16) {texture=/SparksMaps/Environment/Props/StageLights/Textures/T_StageLight_M}.R
    B=TextureSample() {texture=&T_StageLight_M}.G
   )
  )
  n18 = StaticSwitchParameter(
   True=MaterialFunctionCall!(Emissive=@n17, BeatPulse Strength, BeatPulse Min, BeatPulse Max) {materialFunction=/SparksMaps/Environment/MaterialFunctions/MF_MojoBeatPulse}.Result
   False=@n17
  ) {parameterName=UseMojoBeatPulse}
  StaticSwitchParameter(
   True=Multiply(
    A=MaterialFunctionCall() {materialFunction=/SparksStageShow/Systems/StageManager/FXManager/MaterialFunctions/MF_MojoStageShow}.Stage Show Anim
    B=@n18
   )
   False=@n18
  ) {defaultValue=true, parameterName=UseMojoStageShow}
  n19 = TransformPosition(None=Constant3Vector()) {transformSourceType=TRANSFORMPOSSOURCE_Instance, transformType=TRANSFORMPOSSOURCE_World}
  n20 = WorldPosition()
  StaticSwitchParameter(
   True=RotateAboutAxis(
    NormalizedRotationAxis=NamedRerouteUsage()
    RotationAngle=NamedRerouteUsage()
    PivotPoint=NamedRerouteDeclaration(None=@n19) {name=Cone Origin, nodeColor=<rgba 0.0638 0.4605 1 1>}
    Position=Add(A=@n20.XYZ)
   )
   False=Constant3Vector()
  ) {parameterName=UseWpoAnimation}
  TextureSample() {samplerType=SAMPLERTYPE_Normal, texture=&T_MotorizedSpotLight_01_N}
  TextureSample() {texture=/Game/Athena/Environments/Festivus/MotorizedLights/T_MotorizedSpotLight_01_D}
 ]
}
